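# Ambient spot cue for a power plant water pump, built from underground pipe and outpost pump sounds.
# It mixes randomly delayed, repeating pipe rattles, a boosted pipe loop and a quieter steady water pump loop.
# Two nested looping envelopers sweep the pipe loop on a 30-second cycle.
# All quality levels share that mix, played with capsule attenuation, an exterior ambient bus send and limited concurrency.
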SoundCue /Game/Sounds/Helios/Environment_Actors/PowerPlantWaterPump/Spot_PowerPlantWaterPump_01_Cue
root=Modulator_3
nodes [
 {id=Delay_0, children [Random_0], properties {delayMax=6}}
 {
  id=Enveloper_0
  children [WavePlayer_0]
  properties {
   bLoopIndefinitely=true
   loopEnd=30
   pitchCurve {editorCurveData {keys [{time=0}]}}
   volumeCurve {editorCurveData {keys [<at 0 1 RCIM_Cubic>,<at 15 0.25 RCIM_Cubic>,<at 30 1 RCIM_Cubic>]}}
  }
 }
 {
  id=Enveloper_1
  children [Enveloper_0]
  properties {
   bLoopIndefinitely=true
   loopEnd=30
   pitchCurve {editorCurveData {keys [{interpMode=RCIM_Cubic, time=0}, <at 15 0.8 RCIM_Cubic>, {interpMode=RCIM_Cubic, time=30}]}}
   volumeCurve {editorCurveData {keys [<at 0 1 RCIM_Cubic>,<at 15 0.25 RCIM_Cubic>,<at 30 1 RCIM_Cubic>]}}
  }
 }
 {id=Looping_0, children [Delay_0]}
 {id=Mixer_1, children [Modulator_1,Enveloper_1,Modulator_2], properties {inputVolume [0.8,2,0.3]}}
 {id=Modulator_0, children [Looping_0], properties {pitchMax=0.9, pitchMin=0.8}}
 {id=Modulator_1, children [Modulator_0], properties {pitchMin=0.9}}
 {id=Modulator_2, children [WavePlayer_5], properties {volumeMax=1, volumeMin=1}}
 {id=Modulator_3, children [QualityLevel_0], properties {pitchMin=0.9}}
 {id=QualityLevel_0, children [Mixer_1,Mixer_1,Mixer_1,Mixer_1]}
 {id=Random_0, children [WavePlayer_1,WavePlayer_3,WavePlayer_4], properties {weights [1,1,1]}}
 {
  id=WavePlayer_0
  properties {bLooping=true, soundWaveAssetPtr=/Game/Sounds/Helios/Environment_Actors/SewerPipes/Spot_UndergroundPipe_Loop_02}
 }
 {id=WavePlayer_1, properties {soundWaveAssetPtr=/Game/Sounds/Helios/Environment_Actors/SewerPipes/Spot_UndergroundPipe_Rattle_01}}
 {id=WavePlayer_3, properties {soundWaveAssetPtr=/Game/Sounds/Helios/Environment_Actors/SewerPipes/Spot_UndergroundPipe_Rattle_03}}
 {id=WavePlayer_4, properties {soundWaveAssetPtr=/Game/Sounds/Helios/Environment_Actors/SewerPipes/Spot_UndergroundPipe_Rattle_04}}
 {
  id=WavePlayer_5
  properties {bLooping=true, soundWaveAssetPtr=/Game/Sounds/Artemis/Biomes/Grasslands/Grassland_Spots/Spot_Outpost_WaterPump_Loop_01}
 }
]
properties {
 attenuationSettings=/Game/Sounds/Helios/Environment_Actors/SewerPipes/Spot_UndergroundPipe_Loop_Capsule_Att
 bOverrideAttenuation=true
 bOverrideConcurrency=true
 maxDistance=950
 soundClassObject=&Fort_Ambient_Props_Athena
 volumeMultiplier=0.501
 attenuationOverrides {
  attenuationShape=Capsule
  falloffDistance=750
  nonSpatializedRadiusEnd=125
  nonSpatializedRadiusStart=150
  stereoSpread=150
  attenuationShapeExtents=<200 50>
 }
 busSends [{audioBus=null, minSendLevel=0.6, sendLevel=1, soundSourceBus=&AmbientAudio_Exterior_SourceBus}]
 concurrencyOverrides {maxCount=2}
 concurrencySet [&Weapon_Drop_StopOldest]
}
